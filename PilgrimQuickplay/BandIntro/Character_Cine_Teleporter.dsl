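; Cinematic helper actor for the Pilgrim Quickplay band intro that teleports and animates player characters.
; It finds the BandIntro Director component on its playspace, gets the pawn at a cinematic index, and caches that pawn, mesh and anim instance.
; Custom events move the cached pawn to this actor's transform or have the director return it to spawn.
; A separate event has the director play a montage on the current spotlight pawn.
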(event EventBeginPlay)

(event Collision|EventActorBeginOverlap
  (params (OtherActor :type Actor%20Object%20Reference)))

(event EventTick
  (params (DeltaSeconds :type Float%20%28single-precision%29)))

(event Custom|Set%20Character%20To%20Control
  (params (Actor :type Actor%20Object%20Reference))
  ;[0.396 1 0.547] Set this actor as the one we control (used by the event below)
  (bind _asfort_player_pawn (Utilities|Casting|CastToFortPlayerPawn Actor))
  (bind _output_get (Variables|Default|SetFort%20Player%20Pawn _asfort_player_pawn))
  (bind _output_get_2 (Variables|Default|SetFort%20Player%20SK%20Mesh (Variables|/Script/Engine.Character|GetMesh _output_get)))
  (Variables|Default|SetFort%20Player%20Anim%20Instance (Components|SkeletalMesh|GetAnimInstance _output_get_2))
  ;[0.03646 0 0.000303] useful debug
  (Development|PrintString (Utilities|String|Append "Cine: finished setting character to control: " (Utilities|GetDisplayName (Variables|Default|GetFort%20Player%20Pawn)))))

(event Custom|Control%20Character%20From%20Index
  (params (Index :type Integer))
  ;[0.396 1 0.547] Set up the cinematic actor at this index as the one to control
  (Variables|Default|SetCinematic%20Index Index)
  (bind _returnvalue (Class|PlayspaceLibrary|GetPlayspaceforActor self self))
  (Utilities|IsValid _returnvalue
    (:Is%20Valid
      (Utilities|IsValid (Actor|GetComponentbyClass _returnvalue "/PilgrimQuickplay/Playspaces/BP_PilgrimQuickPlay_BandIntro_Director.BP_PilgrimQuickPlay_BandIntro_Director_C")
        (:Is%20Valid
          (bind _actor (Class|BPPilgrimQuickPlayBandIntroDirector|GetPawnatIndex (Actor|GetComponentbyClass _returnvalue "/PilgrimQuickplay/Playspaces/BP_PilgrimQuickPlay_BandIntro_Director.BP_PilgrimQuickPlay_BandIntro_Director_C") (Variables|Default|GetCinematic%20Index)))
          (Utilities|IsValid _actor
            (:Is%20Valid
              (CallFunction|SetCharactertoControl _actor))
            (:Is%20Not%20Valid)))))))

(event Custom|Move%20Character%20Here
  (bind _self self)
  (bind _returnvalue (Transformation|GetActorTransform _self))
  (bind _fort_player_pawn (Variables|Default|GetFort%20Player%20Pawn))
  ;[0.6156 0.3708 1] teleport our controlled actor (if we have one) to our location
  (Utilities|IsValid _fort_player_pawn
    (:Is%20Valid
      (bind (_sweephitresult _returnvalue_2) (Transformation|SetActorTransform _fort_player_pawn _returnvalue false false))
      ;[0.01042 0.000542 0] useful debug
      (Development|PrintString (Utilities|String|Append "Cine: Attempted teleport of " (Utilities|GetDisplayName _fort_player_pawn) " to " (Utilities|String|ToString(Transform) _returnvalue) " : " (Utilities|String|ToString(Boolean) _returnvalue_2))))))

(event Custom|Move%20Character%20To%20Spawn
  ;[0.6156 0.3708 1] teleport our character to its spawn location
  (Utilities|IsValid (Variables|Default|GetFort%20Player%20Pawn)
    (:Is%20Valid
      (bind _returnvalue (Class|PlayspaceLibrary|GetPlayspaceforActor self self))
      (Utilities|IsValid _returnvalue
        (:Is%20Valid
          (Utilities|IsValid (Actor|GetComponentbyClass _returnvalue "/PilgrimQuickplay/Playspaces/BP_PilgrimQuickPlay_BandIntro_Director.BP_PilgrimQuickPlay_BandIntro_Director_C")
            (:Is%20Valid
              (Class|BPPilgrimQuickPlayBandIntroDirector|TeleportPawnAtIndexToSpawnLocation (Actor|GetComponentbyClass _returnvalue "/PilgrimQuickplay/Playspaces/BP_PilgrimQuickPlay_BandIntro_Director.BP_PilgrimQuickPlay_BandIntro_Director_C") (Variables|Default|GetCinematic%20Index)))))))))

(event Custom|Play%20Spotlight%20Montage
  (params (Montage :type Anim%20Montage%20Object%20Reference))
  ;[0.5565 1 0.9301] Play a montage on the current spotlight pawn (we don't have to track the pawn ourselves)
  (bind _returnvalue (Class|PlayspaceLibrary|GetPlayspaceforActor self self))
  (Utilities|IsValid _returnvalue
    (:Is%20Valid
      (Utilities|IsValid (Actor|GetComponentbyClass _returnvalue "/PilgrimQuickplay/Playspaces/BP_PilgrimQuickPlay_BandIntro_Director.BP_PilgrimQuickPlay_BandIntro_Director_C")
        (:Is%20Valid
          (Class|BPPilgrimQuickPlayBandIntroDirector|PlayMontageOnCurrentSpotlightPawn (Actor|GetComponentbyClass _returnvalue "/PilgrimQuickplay/Playspaces/BP_PilgrimQuickPlay_BandIntro_Director.BP_PilgrimQuickPlay_BandIntro_Director_C") Montage))))))

(fn ConstructionScript)
